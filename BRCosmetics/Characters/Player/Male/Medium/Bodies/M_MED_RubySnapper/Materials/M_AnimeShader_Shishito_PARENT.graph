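# Parent anime-style skeletal mesh character material for Shishito, adding animated eyes to the shared 2023 anime shader.
# Chains the shader's Pre, Calculations and Post functions, inserting eye animation on the color path and eyeball masking on DFL lines.
# The eye mask comes from a selectable channel of a texture or vertex color, or from positive UV1 U.
# An animation-driven HasEyeBlinkChanged parameter feeds Temporal Responsiveness to invalidate TSR history when eyes change.
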Material /BRCosmetics/Characters/Player/Male/Medium/Bodies/M_MED_RubySnapper/Materials/M_AnimeShader_Shishito_PARENT
graph {
 properties {bAutomaticallySetUsageInEditor=false, bUseMaterialAttributes=true, bUsedWithSkeletalMesh=true, usedShadingModels=Default Lit}
 nodes [
  n1 = MaterialFunctionCall!(CustomUV) {materialFunction=/Game/Global/Materials/Character_Parent_2023/Anime_Parent_2023/MaterialFunctions/MFC_AnimeShader_2023_Pre}
  n2 = MaterialFunctionCall!(CustomLightVector, OverrideNormal_WS, STT=@n1.STT, Normal_Map=@n1.STT, DFL=@n1.STT, FX=@n1.STT) {materialFunction=/Game/Global/Materials/Character_Parent_2023/Anime_Parent_2023/MaterialFunctions/MFC_AnimeShader_2023_Calculations}
  n3 = VertexColor()
  n4 = TextureSampleParameter2D() {
   group=01_EyeAnimation_Parameters
   parameterName=EyeMask
   samplerSource=SSM_Wrap_WorldGroupSettings
   samplerType=SAMPLERTYPE_Masks
   sortPriority=-32
   texture=&T_WhiteMask
  }
  n5 = StaticSwitchParameter(True=AppendVector(A=@n3, B=@n3), False=@n4.RGBA) {group=01_EyeAnimation_Parameters, parameterName=UseVertexColorForEyeMask, sortPriority=-33}
  n6 = TextureCoordinate() {coordinateIndex=1}
  n7 = Sign(None=Subtract(A=ComponentMask(None=@n6) {r=true}) {constB=0.0001})
  n8 = MaterialFunctionCall!(
   Color Lit=@n1.Color_Lit
   Color Shaded=@n1.Color_Lit
   Use UV2=StaticBoolParameter() {defaultValue=true, group=01_EyeAnimation_Parameters, parameterName=Use UV2 for Eye Animation, sortPriority=-30}
   Mask=StaticSwitchParameter(
    True=ChannelMaskParameter(None=@n5) {group=01_EyeAnimation_Parameters, parameterName=EyeMaskChannel, sortPriority=-32}.R
    False=Saturate(None=@n7)
   ) {group=01_EyeAnimation_Parameters, parameterName=UseEyeMask, sortPriority=-33}
  ) {materialFunction=/Game/Global/Materials/Character_Parent_2023/Anime_Parent_2023/MaterialFunctions/MF_AnimeEyeAnimation_v2}
  MaterialFunctionCall!(
   Shading_Gradient=@n2.Shading_Gradient
   Threshold_Gradient=@n2.Shading_Gradient
   Specular_Gradient=@n2.Shading_Gradient
   Sheen_Mask=@n2.Shading_Gradient
   Thickness=@n2.Shading_Gradient
   DFL_Mask=MaterialFunctionCall!(DFLines=@n2.DFL_Mask) {materialFunction=/BRCosmetics/Characters/Player/Male/Medium/Bodies/M_MED_SilentNovel_Vane/Materials/MF_EyeballsDFLineMasking}.Result
   FX=@n2.Shading_Gradient
   Metallic_Gradient=@n2.Shading_Gradient
   RimLight_Gradient=@n2.Shading_Gradient
   NormalBuffer=@n2.Shading_Gradient
   --------
   CINE_LightSources=@n2.Shading_Gradient
   CINE_LightColor=@n2.Shading_Gradient
   ---------
   Color_Lit=@n8.Lit
   Color_Shaded=@n8.Lit
   SRM=@n1.SRM
   Emissive=@n1.SRM
   DFL_Color=@n1.SRM
   Specular_Color=@n1.SRM
   RimLight_Color=@n1.SRM
  ) {materialFunction=/Game/Global/Materials/Character_Parent_2023/Anime_Parent_2023/MaterialFunctions/MFC_AnimeShader_2023_Post}
  n9 = ScalarParameter() {
   desc="If the base head AnimBP parent's \"Has Material-driven Pupils\" flag is true, this is set to either 0 or 1 depending on whether the pupil look direction has changed this frame or not.\r\nUse this to set the TSR invalidation amount (see the Temporal Responsiveness special material output node's tooltip for more details)"
   group=01_EyeAnimation_AnimData
   parameterName=HasEyeBlinkChanged
  }
  TemporalResponsivenessOutput!(None=Max(B=@n9))
 ]
}
