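# Animation Blueprint for the RubySnapper medium male body cosmetic, adding secondary motion to the pose copied from its attached parent mesh.
# Weighted constraints make belt bones follow pelvis and thighs, feeding a Control Rig driven by Aiming.
# AnimDynamics springs then move holster and zipper bones, before an additive holster-base offset.
# The update event sets Aiming from targeting, and Skydive_Alpha from skydive, emote or crouch states.
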AnimBlueprint /BRCosmetics/Characters/Player/Male/Medium/Bodies/M_MED_RubySnapper/Meshes/M_MED_RubySnapper_AnimBP
target_skeleton=&/M_MED_RubySnapper_Skeleton
graphs [
 AnimGraph {nodes [
   n1 = "Constraint - Belt_r_01 (PP)"!(
    ComponentPose=Local To Component(LocalPose=Copy Pose From Mesh() {bUseAttachedParent=true})
    Alpha=1
    ConstraintWeights_0=0.3
    ConstraintWeights_1:float=0.7
   ) -> (Pose) {
    boneToModify {boneName=Belt_r_01}
    constraintSetup [{targetBone {boneName=pelvis}}, {targetBone {boneName=thigh_r}}]
    constraintWeights [1,1]
   }
   n2 = "Constraint - Belt_l_01 (PP)"!(ComponentPose=@n1, Alpha=1, ConstraintWeights_0=0.7, ConstraintWeights_1:float=0.3) -> (Pose) {
    boneToModify {boneName=Belt_l_01}
    constraintSetup [{targetBone {boneName=pelvis}}, {targetBone {boneName=thigh_l}}]
    constraintWeights [1,1]
   }
   n3 = Control Rig!(Source=Component To Local(ComponentPose=@n2), Alpha, Aiming:double=Get Aiming() -> (Aiming:double)) -> (Pose) {controlRigAssetReference {blueprintRigClass=&/M_MED_RubySnapper_CtrlRig.M_MED_RubySnapper_CtrlRig_C}}
   n4 = Anim Dynamics - Bone: dyn_holster_01(ComponentPose=Local To Component(LocalPose=@n3), Alpha=1, ExternalForce="0,0,0") {
    angularDampingOverride=1
    angularSpringConstant=3500
    bAngularSpring=true
    bOverrideAngularDamping=true
    bUseGravityOverride=true
    gravityScale=0.35
    simulationSpace=RootRelative
    boundBone {boneName=dyn_holster_01}
    componentAppliedLinearAccClamp=<5000 5000 5000>
    componentLinearAccScale=<0.5 0.5 0.5>
    componentLinearVelScale=<0.7 0.7 0.7>
    gravityOverride {z=-980}
    physicsBodyDefinitions [{
      boundBone {boneName=dyn_holster_01}
      boxExtents=<20 30 30>
      constraintSetup {angularLimitsMax=<0 13 13>, angularLimitsMin=<0 -8 -8>, angularTarget=<1 0 0>}
      localJointOffset=<2.8 0 0>
     }]
   }
   n5 = Anim Dynamics - Bone: dyn_zipper(ComponentPose=@n4, Alpha=1) {
    angularDampingOverride=1
    angularSpringConstant=15000
    bAngularSpring=true
    bOverrideAngularDamping=true
    bUseGravityOverride=true
    gravityScale=0.35
    simulationSpace=BoneRelative
    boundBone {boneName=dyn_zipper}
    componentAppliedLinearAccClamp=<6000 6000 6000>
    componentLinearAccScale=<2.5 2.5 2.5>
    componentLinearVelScale=<2.5 2.5 2.5>
    gravityOverride {z=-980}
    physicsBodyDefinitions [{
      boundBone {boneName=dyn_zipper}
      boxExtents=<20 28 25>
      constraintSetup {angularLimitsMax=<0 8 2>, angularLimitsMin=<0 -8 -18>, angularTarget=<1 0 0>}
      localJointOffset=<4 0 0>
     }]
   }
   n6 = "Transform (Modify) Bone - Bone: base_holster_01"!(ComponentPose=@n5, bAlphaBoolEnabled:bool=True, Translation="0,0,1", Rotation="8.5,0,0", Scale="1,1,1") -> (Pose) {
    alphaInputType=Bool
    rotationMode=BMM_Additive
    rotationSpace=BCS_BoneSpace
    translationMode=BMM_Additive
    translationSpace=BCS_BoneSpace
    alphaBoolBlend {blendInTime=0.3, blendOutTime=0.3}
    boneToModify {boneName=base_holster_01}
    rotation {pitch=8.5}
    translation {z=1}
   }
   Output Pose(Result=Component To Local(ComponentPose=@n6)) {layerGroup=None}
  ]}
 EventGraph {nodes [
   Event BlueprintUpdateAnimation()
   > n1 = Sequence()
   > n2 = Branch(Condition=Get bIsTargeting())
   > Set Aiming(Aiming:double=1) -> (Output_Get:double)
   Set Aiming(execute=@n2.else, Aiming=0.45)
   n3 = Branch(execute=@n1.then_1, Condition=Get bIsSkyDiving())
   > Set Skydive_Alpha(Skydive_Alpha:double=0.35) -> (Output_Get:double)
   n4 = Branch(execute=@n3.else, Condition=Get bIsPlayingEmote())
   > Set Skydive_Alpha(Skydive_Alpha=0.25)
   n5 = Branch(execute=@n4.else, Condition=Get bIsCrouching())
   > Set Skydive_Alpha(Skydive_Alpha=0.9)
   Set Skydive_Alpha(execute=@n5.else, Skydive_Alpha=1)
  ]}
]
variables [
 Aiming {type=number}
 Skydive_Alpha {type=number}
]
