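# Animation Blueprint for the RubySnapper male face accessory mesh that adds hair motion on top of the pose copied from its attached parent.
# Chained Anim Dynamics nodes give seven hair bones sprung, angle-limited secondary motion.
# Additive Modify Bone nodes then adjust the bottom center and right hair base bones.
# Each update derives these corrections from head-to-neck rotation, switching between static and sprint sets via alphas set by bIsSprinting.
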AnimBlueprint /BRCosmetics/Characters/Player/Male/Medium/Bodies/M_MED_RubySnapper/Meshes/Parts/M_MED_RubySnapper_FaceAcc_AnimBP
target_skeleton=&/M_MED_RubySnapper_FaceAcc_Skeleton
graphs [
 AnimGraph {nodes [
   n1 = Anim Dynamics - Bone: dyn_hair_c(ComponentPose=Local To Component(LocalPose=Copy Pose From Mesh() {bUseAttachedParent=true}), Alpha=1, ExternalForce="0,0,0") {
    angularDampingOverride=1
    angularSpringConstant=16000
    bAngularSpring=true
    bOverrideAngularDamping=true
    bUseGravityOverride=true
    gravityScale=0.35
    simulationSpace=BoneRelative
    boundBone {boneName=dyn_hair_c}
    componentAppliedLinearAccClamp=<5000 5000 5000>
    componentLinearAccScale=<0.8 0.8 0.8>
    componentLinearVelScale=<1 1 1>
    gravityOverride {z=-780}
    physicsBodyDefinitions [{
      boundBone {boneName=dyn_hair_c}
      boxExtents=<20 30 30>
      constraintSetup {angularLimitsMax=<5 15 8>, angularLimitsMin=<-5 -15 -15>, angularTarget=<1 0 0>}
      localJointOffset=<4 0 0>
     }]
   }
   n2 = Anim Dynamics - Bone: dyn_hair_l_01(ComponentPose=@n1, Alpha=1, ExternalForce="0,0,0") {
    angularDampingOverride=1
    angularSpringConstant=16000
    bAngularSpring=true
    bOverrideAngularDamping=true
    bUseGravityOverride=true
    gravityScale=0.35
    simulationSpace=BoneRelative
    boundBone {boneName=dyn_hair_l_01}
    componentAppliedLinearAccClamp=<5000 5000 5000>
    componentLinearAccScale=<0.8 0.8 0.8>
    componentLinearVelScale=<1 1 1>
    gravityOverride {z=-780}
    physicsBodyDefinitions [{
      boundBone {boneName=dyn_hair_l_01}
      boxExtents=<20 30 30>
      constraintSetup {angularLimitsMax=<5 15 15>, angularLimitsMin=<-5 -5 -15>, angularTarget=<1 0 0>}
      localJointOffset=<4 0 0>
     }]
   }
   n3 = Anim Dynamics - Bone: dyn_hair_l_02(ComponentPose=@n2, Alpha=1, ExternalForce="0,0,0") {
    angularDampingOverride=1
    angularSpringConstant=16000
    bAngularSpring=true
    bOverrideAngularDamping=true
    bUseGravityOverride=true
    gravityScale=0.35
    simulationSpace=BoneRelative
    boundBone {boneName=dyn_hair_l_02}
    componentAppliedLinearAccClamp=<5000 5000 5000>
    componentLinearAccScale=<0.8 0.8 0.8>
    componentLinearVelScale=<1 1 1>
    gravityOverride {z=-780}
    physicsBodyDefinitions [{
      boundBone {boneName=dyn_hair_l_02}
      boxExtents=<20 30 30>
      constraintSetup {angularLimitsMax=<5 15 15>, angularLimitsMin=<-5 -5 -15>, angularTarget=<1 0 0>}
      localJointOffset=<4 0 0>
     }]
   }
   n4 = Anim Dynamics - Bone: dyn_hair_r_01(ComponentPose=@n3, Alpha=1, ExternalForce="0,0,0") {
    angularDampingOverride=1
    angularSpringConstant=16000
    bAngularSpring=true
    bOverrideAngularDamping=true
    bUseGravityOverride=true
    gravityScale=0.35
    simulationSpace=BoneRelative
    boundBone {boneName=dyn_hair_r_01}
    componentAppliedLinearAccClamp=<5000 5000 5000>
    componentLinearAccScale=<0.8 0.8 0.8>
    componentLinearVelScale=<1 1 1>
    gravityOverride {z=-780}
    physicsBodyDefinitions [{
      boundBone {boneName=dyn_hair_r_01}
      boxExtents=<20 30 30>
      constraintSetup {angularLimitsMax=<5 5 15>, angularLimitsMin=<-5 -15 -15>, angularTarget=<1 0 0>}
      localJointOffset=<4 0 0>
     }]
   }
   n5 = Anim Dynamics - Bone: dyn_hair_r_02(ComponentPose=@n4, Alpha=1, ExternalForce="0,0,0") {
    angularDampingOverride=1
    angularSpringConstant=16000
    bAngularSpring=true
    bOverrideAngularDamping=true
    bUseGravityOverride=true
    gravityScale=0.35
    simulationSpace=BoneRelative
    boundBone {boneName=dyn_hair_r_02}
    componentAppliedLinearAccClamp=<5000 5000 5000>
    componentLinearAccScale=<0.8 0.8 0.8>
    componentLinearVelScale=<1 1 1>
    gravityOverride {z=-780}
    physicsBodyDefinitions [{
      boundBone {boneName=dyn_hair_r_02}
      boxExtents=<20 30 30>
      constraintSetup {angularLimitsMax=<5 5 15>, angularLimitsMin=<-5 -15 -15>, angularTarget=<1 0 0>}
      localJointOffset=<4 0 0>
     }]
   }
   n6 = Anim Dynamics - Bone: dyn_hair_bottom_l(ComponentPose=@n5, Alpha=1, ExternalForce="0,0,0") {
    angularDampingOverride=1
    angularSpringConstant=20000
    bAngularSpring=true
    bOverrideAngularDamping=true
    bUseGravityOverride=true
    gravityScale=0.35
    simulationSpace=BoneRelative
    boundBone {boneName=dyn_hair_bottom_l}
    componentAppliedLinearAccClamp=<5000 5000 5000>
    componentLinearAccScale=<0.8 0.8 0.8>
    componentLinearVelScale=<1 1 1>
    gravityOverride {z=-980}
    physicsBodyDefinitions [{
      boundBone {boneName=dyn_hair_bottom_l}
      boxExtents=<20 30 30>
      constraintSetup {angularLimitsMax=<8 15 15>, angularLimitsMin=<-10 -15 -5>, angularTarget=<1 0 0>}
      localJointOffset=<4 0 0>
     }]
   }
   n7 = Anim Dynamics - Bone: dyn_hair_bottom_r(ComponentPose=@n6, Alpha=1, ExternalForce="0,0,0") {
    angularDampingOverride=1
    angularSpringConstant=20000
    bAngularSpring=true
    bOverrideAngularDamping=true
    bUseGravityOverride=true
    gravityScale=0.35
    simulationSpace=BoneRelative
    boundBone {boneName=dyn_hair_bottom_r}
    componentAppliedLinearAccClamp=<5000 5000 5000>
    componentLinearAccScale=<0.8 0.8 0.8>
    componentLinearVelScale=<1 1 1>
    gravityOverride {z=-980}
    physicsBodyDefinitions [{
      boundBone {boneName=dyn_hair_bottom_r}
      boxExtents=<20 30 30>
      constraintSetup {angularLimitsMax=<10 15 18>, angularLimitsMin=<-10 -15 -2>, angularTarget=<1 0 0>}
      localJointOffset=<4 0 0>
     }]
   }
   n8 = "Transform (Modify) Bone - Bone: base_hair_bottom_c"(
    ComponentPose=@n7
    Alpha=Get Static_alpha() -> (Static_alpha:double)
    Translation=Get hair_translate() -> (hair_translate:Vector)
    Rotation=Get hair_rotate_0() -> (hair_rotate_0:Rotator)
    Scale=Get hair_scale() -> (hair_scale:Vector)
   ) {
    rotationMode=BMM_Additive
    rotationSpace=BCS_BoneSpace
    scaleMode=BMM_Additive
    scaleSpace=BCS_BoneSpace
    translationMode=BMM_Additive
    translationSpace=BCS_BoneSpace
    boneToModify {boneName=base_hair_bottom_c}
   }
   n9 = "Transform (Modify) Bone - Bone: base_hair_bottom_c"(
    ComponentPose=@n8
    Alpha=Get Sprint_alpha() -> (Sprint_alpha:double)
    Translation="0,0,0"
    Rotation=Get hair_rotate_1() -> (hair_rotate_1:Rotator)
    Scale=Get hair_scale()
   ) {
    rotationMode=BMM_Additive
    rotationSpace=BCS_BoneSpace
    scaleMode=BMM_Additive
    scaleSpace=BCS_BoneSpace
    translationMode=BMM_Additive
    translationSpace=BCS_BoneSpace
    boneToModify {boneName=base_hair_bottom_c}
   }
   n10 = "Transform (Modify) Bone - Bone: base_hair_bottom_r"(ComponentPose=@n9, Alpha=Get Static_alpha(), Translation="0,0,0", Rotation=Get hair_r_rotate_0() -> (hair_r_rotate_0:Rotator), Scale="1,1,1") {
    rotationMode=BMM_Additive
    rotationSpace=BCS_BoneSpace
    translationMode=BMM_Additive
    translationSpace=BCS_BoneSpace
    boneToModify {boneName=base_hair_bottom_r}
   }
   n11 = "Transform (Modify) Bone - Bone: base_hair_bottom_r"(
    ComponentPose=@n10
    Alpha=Get Sprint_alpha()
    Translation="0,0,0"
    Rotation=Get hair_r_rotate_1() -> (hair_r_rotate_1:Rotator)
    Scale="1,1,1"
   ) {
    rotationMode=BMM_Additive
    rotationSpace=BCS_BoneSpace
    translationMode=BMM_Additive
    translationSpace=BCS_BoneSpace
    boneToModify {boneName=base_hair_bottom_r}
   }
   Output Pose(Result=Component To Local(ComponentPose=@n11)) {layerGroup=None}
  ]}
 EventGraph {nodes [
   Event BlueprintUpdateAnimation()
   n1 = GetOwnerSkeletalMeshComponent()
   n2 = BreakRotator(InRot=BreakTransform(InTransform=GetDeltaTransformFromRefPose(self=@n1, BoneName=head, BaseName=neck_01)).Rotation)
   n3 = MapRangeClamped(Value=@n2.Yaw, InRangeA=0, InRangeB=-33, OutRangeB=9)
   n4 = MapRangeClamped(Value=@n2, InRangeA=2, InRangeB=33, OutRangeB=16)
   n5 = MapRangeClamped(Value=@n2, InRangeA=2, InRangeB=30, OutRangeB=18)
   n6 = MapRangeClamped(Value=@n2.Yaw, InRangeA=5, InRangeB=-35, OutRangeB=50)
   > Set hair_rotate_0(hair_rotate_0:Rotator=MakeRotator(Roll=float + float(A=@n3, B=@n4), Yaw=float + float(A=@n6, B=@n5))) -> (Output_Get:Rotator)
   > Set hair_rotate_1(hair_rotate_1:Rotator=MakeRotator(Yaw=MapRangeClamped(Value=@n2.Yaw, InRangeA=0, InRangeB=-33, OutRangeB=49))) -> (Output_Get:Rotator)
   n7 = MapRangeClamped(Value=@n2.Yaw, InRangeA=0, InRangeB=-31, OutRangeB=3)
   > Set hair_translate(hair_translate:Vector=MakeVector(X=@n7, Y=@n7)) -> (Output_Get:Vector)
   > Set hair_r_rotate_0(hair_r_rotate_0:Rotator=MakeRotator(Yaw=MapRangeClamped(Value=@n2.Yaw, InRangeA=0, InRangeB=-31, OutRangeB=24))) -> (Output_Get:Rotator)
   > Set hair_r_rotate_1(hair_r_rotate_1:Rotator=MakeRotator(Yaw=MapRangeClamped(Value=@n2.Yaw, InRangeA=0, InRangeB=-33, OutRangeB=15))) -> (Output_Get:Rotator)
   > Set hair_scale(hair_scale:Vector=MakeVector(X=MapRangeClamped(Value=@n2.Yaw, InRangeA=0, InRangeB=-31, OutRangeA=1, OutRangeB=0.4), Y=1, Z=1)) -> (Output_Get:Vector)
   > n8 = Branch(Condition=Get bIsSprinting())
   > Set Sprint_alpha(Sprint_alpha:double=1) -> (Output_Get:double)
   > Set Static_alpha(Static_alpha:double=0) -> (Output_Get:double)
   Set Sprint_alpha(execute=@n8.else, Sprint_alpha=0)
   > Set Static_alpha(Static_alpha=1)
   Reroute Node(InputPin:Wildcard) -> (OutputPin:Wildcard)
  ]}
]
variables [
 hair_r_rotate_0 {type=Rotator}
 hair_r_rotate_1 {type=Rotator}
 hair_rotate_0 {type=Rotator}
 hair_rotate_1 {type=Rotator}
 hair_scale {type=Vector}
 hair_translate {type=Vector}
 Sprint_alpha {type=number}
 Static_alpha {type=number}
]
